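# Material function for the Aglet05 cosmetic skin on the ZRF vehicle body, overriding attributes from MF_Vehicle_Body_Base.
# An RGBA mask texture and rerouted inputs drive layered color blends of rotated stripe-and-dot patterns, panning gradient stripes and fixed tints.
# Two counter-rotated samples of the pattern texture multiply into a faux halftone that blends between layers.
# The saturated result sets base color, and its clamped, scaled desaturation sets roughness.
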MaterialFunction /VehicleCosmetics/Bodies/ZRF/Skins/Aglet05/MF_Skin_ZRF_Aglet05
graph {nodes [
  n1 = OneMinus(None=NamedRerouteUsage())
  n2 = NamedRerouteUsage()
  n3 = Constant2Vector() {desc=UV Tiling, g=8, r=8}
  n4 = TextureSample(UVs=MaterialFunctionCall(UVs=Multiply(A=@n2, B=@n3), "Rotation Angle (0-1)"=Constant() {r=0.6}) {materialFunction=&CustomRotator}.Rotated Values) {texture=/VehicleCosmetics/SharedTextures/T_Stripes_Dots_RGB}
  n5 = NamedRerouteUsage()
  n6 = TextureSample(UVs=MaterialFunctionCall(UVs=Panner(Coordinate=@n5) {speedY=0.4}, "Rotation Angle (0-1)"=Constant() {r=-0.65}) {materialFunction=&CustomRotator}.Rotated Values) {texture=/VehicleCosmetics/Textures/UniversalTextures/T_GradientStripe}
  n7 = Multiply(A=ConstantBiasScale(None=@n6.RGB) {bias=-0.7, scale=0.3}, B=ComponentMask(None=NamedRerouteUsage()) {r=true})
  n8 = NamedRerouteUsage()
  n9 = TextureSample(UVs=Multiply(A=Panner(Coordinate=@n8) {speedY=0.005}) {constB=35}) {texture=&T_Stripes_Dots_RGB}
  n10 = Multiply(A=@n9.G, B=Clamp(None=ConstantBiasScale(None=@n6.RGB) {bias=-0.7, scale=5}) {minDefault=0.2})
  n11 = NamedRerouteUsage()
  n12 = NamedRerouteUsage()
  n13 = Constant2Vector() {desc=UV Tiling, g=15, r=15}
  n14 = MaterialFunctionCall(UVs=Multiply(A=@n12, B=@n13), "Rotation Angle (0-1)"=Constant() {r=0.6}) {materialFunction=&CustomRotator}
  n15 = TextureSample(UVs=MaterialFunctionCall(UVs=Multiply(A=@n12, B=@n13), "Rotation Angle (0-1)"=Constant() {r=-0.6}) {materialFunction=&CustomRotator}.Rotated Values) {texture=&T_Stripes_Dots_RGB}
  n16 = OneMinus(None=Multiply(
    A=TextureSample(UVs=@n14.Rotated Values) {texture=&T_Stripes_Dots_RGB}.R
    B=Clamp(None=@n15.R) {maxDefault=0.5}
   ) {desc=Faux Halftone})
  n17 = LinearInterpolate(
   A=LinearInterpolate(A=LinearInterpolate(Alpha=@n1) {desc=Basic Color}, B=ConstantBiasScale(None=@n4.R) {bias=-0.3, scale=5}, Alpha=Saturate(None=@n7))
   B=LinearInterpolate(A=Constant3Vector() {constant=<rgba 0.4076 0.4149 0.4123 1>}, B=@n9.G, Alpha=Saturate(None=@n10))
   Alpha=Saturate(None=Multiply(A=@n11, B=@n16))
  )
  n18 = Constant3Vector() {constant=<rgba 0.5955 0.2227 0.3645 1>}
  n19 = NamedRerouteUsage()
  n20 = Saturate(None=LinearInterpolate(A=LinearInterpolate(A=@n17, B=@n18, Alpha=@n19), B=Constant3Vector(), Alpha=NamedRerouteUsage()))
  n21 = Desaturation(None=@n20)
  n22 = Multiply(A=Clamp(None=Multiply(A=@n21) {constB=0.8}) {maxDefault=0.9, minDefault=0.5}) {constB=0.6}
  FunctionOutput(None=SetMaterialAttributes!(
    MaterialAttributes=MaterialFunctionCall() {materialFunction=/VehicleCosmetics/Materials/MaterialFunctions/MF_Vehicle_Body_Base}.Result
    BaseColor=@n20
    Roughness=Saturate(None=@n22)
   ) {attributeSetTypes [#1, #2]})
  NamedRerouteDeclaration(None=TextureCoordinate() {coordinateIndex=1}) {name=TexCoord1_Out, nodeColor=<rgba 0 1 0.3438 1>}
  n23 = TextureSample() {samplerType=SAMPLERTYPE_Masks, texture=&/T_Aglet05_RGB}
  NamedRerouteDeclaration(None=@n23.R) {name=R_Out, nodeColor=<rgba 1 0.7734 0 1>}
  NamedRerouteDeclaration(None=@n23.G) {name=G_Out, nodeColor=<rgba 0.1081 1 0 1>}
  NamedRerouteDeclaration(None=@n23.B) {name=B_Out, nodeColor=<rgba 0 0.1319 1 1>}
  NamedRerouteDeclaration(None=@n23.A) {name=A_Out, nodeColor=<rgba 0.276 0.276 0.276 1>}
 ]}
